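# Unlit, two-sided translucent Niagara material for the swirling blizzard area storm in the ViperBucket emote.
# Panning, UV-distorted wind and frost textures combine with scrolling snow bits and noise, simplified on lower-end platforms.
# Color blends dark, mid and light tones, tinted by time of day and sun direction.
# Opacity uses particle-driven erosion, border, fresnel, depth and camera fades, while masked world position offset pushes vertices along normals.
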Material /BRCosmetics/Animation/Game/MainPlayer/Emotes/ViperBucket/VFX/M_ViperBucket_BlizzardAreaStorm
graph {
 properties {
  bAutomaticallySetUsageInEditor=false
  bUsedWithNiagaraMeshParticles=true
  bUsedWithNiagaraRibbons=true
  bUsedWithNiagaraSprites=true
  blendMode=BLEND_Translucent
  shadingModel=MSM_Unlit
  twoSided=true
  usedShadingModels=Unlit
 }
 nodes [
  Composite() {subgraphName=BorderSquareFade}
  Composite() {subgraphName=Radial_Mask}
  Composite() {subgraphName=BorderRadialFade}
  n1 = Power(Base=NamedRerouteUsage()) {constExponent=3}
  n2 = Saturate(None=Power(
    Base=OneMinus(None=@n1)
    Exp=DynamicParameter() {defaultValue=<rgba 1 2 0 1>, paramNames [WPO_Direction, Fresnel_Intensity, Erosion, WPO_Intensity]}.Fresnel_Intensity
   ))
  MaterialFunctionCall(In=@n2, Contrast=Constant() {r=1}) {materialFunction=&CheapContrast}
  n3 = NamedRerouteUsage()
  n4 = ScalarParameter() {group=Opacity_MeshFresnel, parameterName=UseMeshFresnel?, sortPriority=1}
  n5 = ParticleColor()
  n6 = Multiply(A=Multiply(A=@n5.A, B=NamedRerouteUsage()), B=NamedRerouteUsage())
  n7 = MaterialFunctionCall(Distance=ScalarParameter() {parameterName=CameraFadeNear_Distance}) {materialFunction=&Near_Camera_Fade}
  n8 = Saturate(None=Multiply(A=Power(Base=@n7.Result) {constExponent=5}) {constB=7})
  n9 = Multiply(
   A=Multiply(A=LinearInterpolate(B=@n3, Alpha=@n4) {constA=1}, B=Multiply(A=@n6, B=@n8)) {constA=1}
   B=MaterialFunctionCall("FadeSensitivity (optional)"=ScalarParameter() {parameterName=AngleFade_Sensitivity}) {materialFunction=&FadeBasedOnViewAngle}.Result
  )
  n10 = ScalarParameter() {group=OpacityBorder, parameterName=TexEdge_SS_Min}
  n11 = ScalarParameter() {defaultValue=0.05, group=OpacityBorder, parameterName=TexEdge_SS_Max, sortPriority=33}
  n12 = TextureCoordinate()
  n13 = ComponentMask(None=@n12) {r=true}
  n14 = SmoothStep(
   Min=ScalarParameter() {group=OpacityBorder, parameterName=Fade_Head_SS_min, sortPriority=11}
   Max=ScalarParameter() {defaultValue=0.5, group=OpacityBorder, parameterName=Fade_Head_SS_max, sortPriority=12}
   Value=@n13
  )
  n15 = SmoothStep(
   Min=ScalarParameter() {group=OpacityBorder, parameterName=Fade_Tail_SS_min, sortPriority=13}
   Max=ScalarParameter() {defaultValue=0.5, group=OpacityBorder, parameterName=Fade_Tail_SS_max, sortPriority=14}
   Value=OneMinus(None=@n13)
  )
  n16 = ComponentMask(None=@n12) {g=true}
  n17 = SmoothStep(
   Min=ScalarParameter() {group=OpacityBorder, parameterName=Fade_Top_SS_Min, sortPriority=15}
   Max=ScalarParameter() {defaultValue=0.5, group=OpacityBorder, parameterName=Fade_Top_SS_Max, sortPriority=16}
   Value=@n16
  )
  n18 = SmoothStep(
   Min=ScalarParameter() {group=OpacityBorder, parameterName=Fade_Bottom_SS_Min, sortPriority=17}
   Max=ScalarParameter() {defaultValue=0.5, group=OpacityBorder, parameterName=Fade_Bottom_SS_Max, sortPriority=18}
   Value=OneMinus(None=@n16)
  )
  n19 = Multiply(A=Multiply(A=Saturate(None=@n14), B=Saturate(None=@n15)), B=Multiply(A=Saturate(None=@n17), B=Saturate(None=@n18)))
  MaterialExpressionReroute_12 = Reroute(None=@n19)
  n20 = Length(None=OneMinus(None=TextureCoordinate() {uTiling=2, vTiling=2})) {is=class MaterialExpressionLength}
  n21 = Saturate(None=SmoothStep(
    Min=ScalarParameter() {group=OpacityBorder, parameterName=BorderEdge_RadialMask_SS_Min, sortPriority=1}
    Max=ScalarParameter() {defaultValue=1, group=OpacityBorder, parameterName=BorderEdge_RadialMask_SS_Max, sortPriority=2}
    Value=OneMinus(None=@n20)
   ))
  MaterialExpressionReroute_14 = Reroute(None=@n21)
  n22 = OneMinus(None=TextureCoordinate() {uTiling=2, vTiling=2})
  n23 = Saturate(None=SmoothStep(
    Min=ScalarParameter() {group=OpacityBorder, parameterName=RadialFade_Center_SS_min, sortPriority=5}
    Max=ScalarParameter() {defaultValue=1, group=OpacityBorder, parameterName=RadialFade_Center_SS_max, sortPriority=6}
    Value=Length(None=@n22)
   ))
  MaterialExpressionReroute_13 = Reroute(None=@n23)
  n24 = Multiply(
   A=LinearInterpolate(
    A=@MaterialExpressionReroute_12.Output
    B=@MaterialExpressionReroute_14.Output
    Alpha=ScalarParameter() {group=OpacityBorder, parameterName=isBorderFade_Square_or_Radial?, sortPriority=0}
   )
   B=@MaterialExpressionReroute_13.Output
  )
  n25 = MaterialFunctionCall(Base=@n24, Blend=NamedRerouteUsage()) {materialFunction=&Blend_Overlay}
  n26 = Saturate(None=LinearInterpolate(
    B=SmoothStep(Min=@n10, Max=@n11, Value=@n25.Result) {constMax=0.05}
    Alpha=ScalarParameter() {group=OpacityBorder, parameterName=UseTexEdge?, sortPriority=31}
   ) {constA=1})
  Multiply(
   A=Multiply(A=Saturate(None=@n9), B=Multiply(A=@n26, B=@n24))
   B=LinearInterpolate(B=VertexColor(), Alpha=ScalarParameter() {group=Opacity, parameterName=useVertexColor?}) {constA=1}
  )
  n27 = DepthFade(FadeDistance=ScalarParameter() {defaultValue=200, parameterName=DepthFadeDistance})
  n28 = Constant() {r=1}
  n29 = QualitySwitch(Default=@n27, Low=@n28, High=@n27, Medium=@n27, Epic=@n27)
  NamedRerouteDeclaration(None=ShadingPathSwitch(Default=ShadingPathSwitch(Default=@n29, Deferred=@n29, Mobile=@n28) {desc=Migrated from FeatureLevelSwitch}, Forward=@n28)) {name=Depth Fade, nodeColor=<rgba 1 0.1577 0 1>}
  n30 = MaterialFunctionCall(Float2=TextureCoordinate()) {materialFunction=&BreakOutFloat2Components}
  n31 = Power(Base=Sine(None=Multiply(A=@n30.G) {constB=0.5})) {constExponent=1}
  NamedRerouteDeclaration(None=Saturate(None=Multiply(A=@n31) {constB=2})) {name=TopBottomCreasesFade, nodeColor=<rgba 0.5516 0.0833 1 1>}
  n32 = MaterialFunctionCall(Float2=TextureCoordinate()) {materialFunction=&BreakOutFloat2Components}
  n33 = Multiply(
   A=Power(Base=NamedRerouteUsage(), Exp=ScalarParameter() {defaultValue=2, parameterName=Erosion_Power}) {constExponent=3}
   B=ScalarParameter() {defaultValue=1, parameterName=Erosion_Hardness}
  ) {constB=2}
  n34 = Multiply(
   A=Sine(None=Multiply(A=@n32.G) {constB=0.5})
   B=MaterialFunctionCall(In=Saturate(None=@n33), Contrast=ScalarParameter() {parameterName=Erosion_Extra_Contrast}) {materialFunction=&CheapContrast}.Result
  )
  n35 = LinearInterpolate(Alpha=DynamicParameter() {defaultValue=<rgba 1 2 0 1>, paramNames [WPO_Direction, Fresnel_Intensity, Erosion, WPO_Intensity]}.Erosion) {constA=-1}
  NamedRerouteDeclaration(None=Saturate(None=Subtract(A=@n34, B=@n35))) {name=Erosion, nodeColor=<rgba 0.491 1 0.0616 1>}
  n36 = NamedRerouteUsage()
  n37 = VertexNormalWS()
  n38 = CameraVectorWS()
  n39 = Saturate(None=Abs(None=DotProduct(A=@n37, B=@n38)))
  n40 = SmoothStep(
   Min=ScalarParameter() {group=Opacity_MeshFresnel, parameterName=MeshFresnel_CenterFade_SS_min, sortPriority=2}
   Max=ScalarParameter() {defaultValue=0.5, group=Opacity_MeshFresnel, parameterName=MeshFresnel_CenterFade_SS_max, sortPriority=3}
   Value=OneMinus(None=@n39)
  )
  n41 = SmoothStep(
   Min=ScalarParameter() {group=Opacity_MeshFresnel, parameterName=MeshFresnel_Edge_SS_min, sortPriority=4}
   Max=ScalarParameter() {defaultValue=1, group=Opacity_MeshFresnel, parameterName=MeshFresnel_Edge_SS_max, sortPriority=5}
   Value=@n39
  )
  NamedRerouteDeclaration(None=Multiply(A=Multiply(A=@n36, B=@n40), B=Saturate(None=@n41))) {name=Mesh_Fresnel Center/Edge Fade, nodeColor=<rgba 1 0 0.5859 1>}
  n42 = Constant()
  n43 = NamedRerouteUsage()
  n44 = MaterialFunctionCall(Float2=TextureCoordinate()) {materialFunction=&BreakOutFloat2Components}
  n45 = Multiply(A=@n44.R, B=ScalarParameter() {defaultValue=7, group=SnowBits, parameterName=SnowBits_TileU, sortPriority=1}) {constB=7}
  n46 = Time()
  n47 = Multiply(A=@n46, B=ScalarParameter() {defaultValue=2.6, group=SnowBits, parameterName=SnowBits1_U_Speed}) {constB=3}
  n48 = Multiply(A=@n44.G, B=ScalarParameter() {defaultValue=7, group=SnowBits, parameterName=SnowBits_TileV, sortPriority=2}) {constB=4}
  n49 = Multiply(A=@n46, B=ScalarParameter() {defaultValue=0.35, group=SnowBits, parameterName=SnowBits1_V_Speed}) {constB=0.4}
  n50 = TextureSampleParameter2D(UVs=AppendVector(A=Add(A=@n45, B=@n47), B=Add(A=@n48, B=@n49))) {
   parameterName=TexSnowBits
   samplerSource=SSM_Wrap_WorldGroupSettings
   samplerType=SAMPLERTYPE_LinearGrayscale
   texture=/Game/Effects/Fort_Effects/Textures/Rocks/T_Tiling_Dust_Dirt_Bits_M
  }
  n51 = Power(Base=@n50.R, Exp=ScalarParameter() {defaultValue=4, group=SnowBits, parameterName=TexSnow_Pow}) {constExponent=4}
  n52 = MaterialFunctionCall(Float2=Add(A=TextureCoordinate()) {constB=0}) {materialFunction=&BreakOutFloat2Components}
  n53 = Multiply(A=@n52.R) {constB=7}
  n54 = Time()
  n55 = Multiply(A=@n54, B=ScalarParameter() {defaultValue=3.3, parameterName=SnowBits2_U_Speed}) {constB=4}
  n56 = Multiply(A=@n52.G) {constB=5}
  n57 = Multiply(A=@n54, B=ScalarParameter() {defaultValue=0.9, parameterName=SnowBits2_V_Speed}) {constB=1.2}
  n58 = TextureSampleParameter2D(UVs=AppendVector(A=Add(A=@n53, B=@n55), B=Add(A=@n56, B=@n57))) {parameterName=TexSnowBits2, samplerSource=SSM_Wrap_WorldGroupSettings, texture=/Game/Effects/Fort_Effects/Textures/Tile/T_Dots_Small}
  n59 = Constant() {r=1}
  n60 = MaterialFunctionCall(Float2=TextureCoordinate()) {materialFunction=&BreakOutFloat2Components}
  n61 = Multiply(A=@n60.R) {constB=2}
  n62 = Time()
  n63 = Multiply(A=@n62, B=ScalarParameter() {defaultValue=-1, group=Noise, parameterName=Noise_PanSpeed_U}) {constB=-1.2}
  n64 = Multiply(A=@n60.G)
  n65 = Multiply(A=@n62, B=ScalarParameter() {defaultValue=0.5, group=Noise, parameterName=Noise_PanSpeed_V}) {constB=0.7}
  n66 = TextureSampleParameter2D(UVs=AppendVector(A=Add(A=@n61, B=@n63), B=Add(A=@n64, B=@n65))) {
   group=Noise
   parameterName=Noise
   samplerSource=SSM_Wrap_WorldGroupSettings
   samplerType=SAMPLERTYPE_LinearColor
   texture=/BRCosmetics/Weapons/FORT_Melee/Pickaxe_RhombCamo/FX/T_Pickaxe_RhombCamo_Perlin
  }
  n67 = Power(Base=@n66.R) {constExponent=1.5}
  n68 = QualitySwitch(Default=@n67, Low=@n59, High=@n67, Medium=@n59, Epic=@n67)
  n69 = ShadingPathSwitch(Default=@n68, Deferred=@n68, Mobile=@n59) {desc=Migrated from FeatureLevelSwitch}
  n70 = Multiply(A=Max(A=Saturate(None=@n51), B=LinearInterpolate(Alpha=@n58.R) {constB=0.6}) {constB=0}, B=@n69)
  n71 = Saturate(None=Multiply(A=Power(Base=@n70) {constExponent=1}))
  n72 = LinearInterpolate(Alpha=Saturate(None=Multiply(A=@n43, B=@n71))) {constB=3}
  n73 = QualitySwitch(Default=@n72, Low=@n42, High=@n72, Medium=@n42, Epic=@n72)
  NamedRerouteDeclaration(None=ShadingPathSwitch(Default=@n73, Deferred=@n73, Mobile=@n42) {desc=Migrated from FeatureLevelSwitch}) {name=Snow Bits, nodeColor=<rgba 0.1146 1 1 1>}
  n74 = AppendVector(
   A=ScalarParameter() {group=uvDistort, parameterName=uvDistort_OffsetU, sortPriority=4}
   B=ScalarParameter() {group=uvDistort, parameterName=uvDistort_OffsetV, sortPriority=5}
  )
  n75 = Multiply(
   A=TextureCoordinate()
   B=AppendVector(
    A=ScalarParameter() {defaultValue=1, group=uvDistort, parameterName=uvDistort_ScaleU, sortPriority=2}
    B=ScalarParameter() {defaultValue=1, group=uvDistort, parameterName=uvDistort_ScaleV, sortPriority=3}
   )
  )
  n76 = Time()
  n77 = AppendVector(
   A=ScalarParameter() {group=uvDistort, parameterName=uvDistort_speedU, sortPriority=6}
   B=ScalarParameter() {group=uvDistort, parameterName=uvDistort_speedV, sortPriority=7}
  )
  n78 = TextureSampleParameter2D(UVs=Add(A=Add(A=@n74, B=@n75), B=Multiply(A=@n76, B=@n77))) {
   group=uvDistort
   parameterName=Tex_uvDistort_normalmap
   samplerType=SAMPLERTYPE_Normal
   sortPriority=0
   texture=/Game/Athena/Prototype/HardSurface/Textures/HS_FlatNormal
  }
  n79 = ScalarParameter() {group=uvDistort, parameterName=uvDistort, sortPriority=1}
  n80 = MaterialFunctionCall(Input=ComponentMask(None=@n78.RGB) {g=true, r=true}, Target Low=Multiply(A=Constant() {r=-1}, B=@n79), Target High=@n79) {materialFunction=&RemapValueRange}
  n81 = MaterialFunctionCall(Float2=TextureCoordinate()) {materialFunction=&BreakOutFloat2Components}
  n82 = Time()
  n83 = Add(
   A=Multiply(A=@n81.R, B=ScalarParameter() {defaultValue=1, group=Main Texture, parameterName=Main_Tex_Scale_U})
   B=Multiply(A=ScalarParameter() {defaultValue=1, group=Main Texture, parameterName=Main_Tex_PanSpeed_U}, B=@n82) {constB=1.45}
  )
  n84 = Add(
   A=Multiply(A=@n81.G, B=ScalarParameter() {defaultValue=3, group=Main Texture, parameterName=Main_Tex_Scale_V}) {constB=3}
   B=Multiply(A=@n82, B=ScalarParameter() {defaultValue=0.2, group=Main Texture, parameterName=Main_Tex_PanSpeed_V}) {constB=0.33}
  )
  n85 = TextureSampleParameter2D(UVs=Add(A=NamedRerouteDeclaration(None=@n80.Result) {name=reroute_uvDistort, nodeColor=<rgba 0.5703 0 1 1>}, B=AppendVector(A=@n83, B=@n84))) {group=Main Texture, parameterName=MainTex, samplerSource=SSM_Wrap_WorldGroupSettings, texture=/Game/Effects/FogVolume/noise_flamy_Tile}
  n86 = LinearInterpolate(
   A=@n85.R
   B=MaterialFunctionCall(In=@n85.R, Contrast=Constant() {r=0.5}) {materialFunction=&CheapContrast}.Result
  ) {constAlpha=0.6}
  n87 = LinearInterpolate(Alpha=@n86) {constA=0.1}
  n88 = AppendVector(
   A=Multiply(A=@n81.R, B=ScalarParameter() {defaultValue=1, group=Frost Texture, parameterName=FrostTex_Scale_U})
   B=Multiply(A=@n81.G, B=ScalarParameter() {defaultValue=2, group=Frost Texture, parameterName=FrostTex_Scale_V}) {constB=2}
  )
  n89 = AppendVector(
   A=ScalarParameter() {defaultValue=-0.45, group=Frost Texture, parameterName=FrostTex_PanSpeed_U}
   B=ScalarParameter() {defaultValue=0.15, group=Frost Texture, parameterName=FrostTex_PanSpeed_V}
  )
  n90 = TextureSampleParameter2D(UVs=Add(A=Panner(Coordinate=@n88, Speed=@n89) {speedX=-0.7, speedY=0.33}, B=NamedRerouteUsage())) {
   group=Frost Texture
   parameterName=TexFrost
   samplerSource=SSM_Wrap_WorldGroupSettings
   texture=/Game/Effects/Fort_Effects/Textures/Smoke/T_Clouds_01
  }
  n91 = Constant() {r=0.1}
  n92 = LinearInterpolate(
   A=@n86
   B=Saturate(None=MaterialFunctionCall(In=@n90.R, Contrast=@n91) {materialFunction=&CheapContrast}.Result)
   Alpha=ScalarParameter() {defaultValue=0.2, parameterName=TexLerp}
  ) {constAlpha=0.2}
  n93 = Multiply(A=MaterialFunctionCall("Xb1,PS4"=@n92, "PC-High&&Epic,XSX,PS5"=@n92, "PC-Perf,Android"=@n87, Switch=@n87, "PC-Low(NotPerf)&&Medium"=@n92) {materialFunction=&PlatformSwitch_1D}.Result) {constB=1.55}
  NamedRerouteDeclaration(None=LinearInterpolate(B=Saturate(None=@n93), Alpha=NamedRerouteUsage()) {constA=1}) {name=Spin Wind Texture, nodeColor=<rgba 0.1255 0.8195 1 1>}
  n94 = NamedRerouteUsage()
  n95 = NamedRerouteUsage()
  n96 = Multiply(A=Saturate(None=Multiply(A=@n94, B=@n95)), B=NamedRerouteUsage())
  n97 = Multiply(
   A=LinearInterpolate(Alpha=Saturate(None=@n96)) {constA=-0.2}
   B=ScalarParameter() {defaultValue=1, group=WPO, parameterName=WPO_Intensity, sortPriority=1}
  ) {constB=85}
  n98 = VectorParameter() {group=WPO, parameterName=WPO_Direction, defaultValue=<rgba 1 1 0 1>}
  NamedRerouteDeclaration(None=Multiply(A=Multiply(A=@n97, B=@n98.RGB), B=VertexNormalWS())) {name=WPO, nodeColor=<rgba 1 0.4453 0 1>}
  NamedRerouteDeclaration(None=Multiply(A=NamedRerouteUsage(), B=@n2)) {name=Opacity, nodeColor=<rgba 0 1 0.8125 1>}
  NamedRerouteUsage()
  PerInstanceRandom()
  PinBase()
  PinBase!(Output=@n19) {reroutePins [Output {expression=&self:MaterialExpressionReroute_12}]}
  PinBase!()
  PinBase!(Output=@n23) {reroutePins [Output {expression=&self:MaterialExpressionReroute_13}]}
  PinBase!()
  PinBase!(Output=@n21) {reroutePins [Output {expression=&self:MaterialExpressionReroute_14}]}
  n99 = VectorParameter() {parameterName=LightColorMorning, defaultValue=<rgba 0.9 0.495 0.5288 1>}
  n100 = VectorParameter() {parameterName=LightColorDay, defaultValue=<rgba 0.55 0.8643 1 1>}
  n101 = VectorParameter() {parameterName=LightColorEvening, defaultValue=<rgba 0.85 0.6588 0.595 1>}
  n102 = VectorParameter() {parameterName=LightColorNight, defaultValue=<rgba 0.36 0.585 0.9 1>}
  n103 = NamedRerouteUsage()
  n104 = OneMinus(None=@n103)
  n105 = LinearInterpolate(
   A=VectorParameter() {group=Color, parameterName=ColorMid, sortPriority=35, defaultValue=<rgba 0.0988 0.76 0.76 1>}.RGB
   B=VectorParameter() {group=Color, parameterName=ColorDark, sortPriority=40, defaultValue=<rgba 0.01 0.0625 0.1 1>}.RGB
   Alpha=MaterialFunctionCall(In=Power(Base=@n104) {constExponent=2.5}, Contrast=Constant() {r=0.1}) {materialFunction=&CheapContrast}.Result
  )
  n106 = Saturate(None=Power(Base=@n103) {constExponent=1.5})
  n107 = VectorParameter() {group=Color, parameterName=ColorLight, sortPriority=30, defaultValue=<rgba 1 1 1 1>}
  n108 = NamedRerouteUsage()
  n109 = Multiply(
   A=NamedRerouteUsage()
   B=LinearInterpolate(B=LinearInterpolate(Alpha=@n69) {constA=0.4}, Alpha=ScalarParameter() {defaultValue=1, parameterName=Use_Color_Noise}) {constA=1}
  )
  n110 = Multiply(A=Multiply(A=Max(A=@n108, B=@n109), B=@n94)) {constB=3}
  n111 = Multiply(A=LinearInterpolate(A=LinearInterpolate(A=@n105, B=@n107.RGB, Alpha=@n106), B=@n107.RGB, Alpha=Saturate(None=@n110)), B=@n5.RGB)
  n112 = Constant() {r=0.4}
  n113 = CollectionParameter() {
   collection=&FortniteMaterialParameters
   parameterName=SunlightDirectionalVector
  }
  n114 = Transform(None=Normalize(VectorInput=ComponentMask(None=@n113) {b=true, g=true, r=true})) {transformSourceType=TRANSFORMSOURCE_World}
  n115 = VertexNormalWS()
  n116 = Power(Base=ConstantBiasScale(None=DotProduct(A=@n114, B=@n115))) {constExponent=7}
  n117 = LinearInterpolate(
   A=Multiply(
    A=MaterialFunctionCall(Color_Morning=@n99.RGB, Color_Day=@n100.RGB, Color_Evening=@n101.RGB, Color_Night=@n102.RGB) {materialFunction=&Color_Time_of_Day}.Result
    B=Desaturation(None=@n111, Fraction=@n112)
   ) {constA=0.5}
   B=@n111
   Alpha=LinearInterpolate(A=@n116, B=OneMinus(None=@n116), Alpha=ScalarParameter() {parameterName=Invert_Normal_ToD})
  )
  n118 = Max(A=NamedRerouteUsage(), B=@n117)
  Reroute(None=MaterialFunctionCall(Emissive=LinearInterpolate(A=@n118, B=@n117)) {materialFunction=&MF_FX_EmissiveAdjustments}.Result)
  Saturate(None=@n78.R)
 ]
}
