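# Masked subsurface material for snow Niagara sprites in the ViperBucket emote, built on dirt-clod cluster textures.
# A height texture drives its own bump-offset UVs, and the resampled red channel feeds a subsurface thickness multiplier and a smoothstep erosion mask.
# Particle alpha, remapped in inverse, sets the erosion threshold.
# A normal map, blue-tinted subsurface color and scalar roughness, specular and metallic parameters define the surface.
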Material /BRCosmetics/Animation/Game/MainPlayer/Emotes/ViperBucket/VFX/M_ViperBucket_Snow
graph {
 properties {
  bUsedWithNiagaraSprites=true
  blendMode=BLEND_Masked
  pixelDepthOffsetMode=PDOM_Legacy
  shadingModel=MSM_Subsurface
  usedShadingModels=Subsurface
 }
 nodes [
  DynamicParameter() {defaultValue=<rgba 0 0 1 1>, paramNames [Erode,Param2,Param3,Param4]}
  n1 = TextureSampleParameter2D() {parameterName=HeightMap_Alpha, texture=/Game/Effects/Fort_Effects/Textures/Dirt/T-DirtClodCluster_D}
  n2 = ScalarParameter() {defaultValue=1, parameterName=Rch_or_Ach}
  n3 = TextureSampleParameter2D(UVs=BumpOffset(Height=LinearInterpolate(A=@n1.R, B=@n1.R, Alpha=@n2), HeightRatioInput=ScalarParameter() {defaultValue=0.05, parameterName=BumpOffset})) {parameterName=HeightMap_Alpha, texture=&T-DirtClodCluster_D}
  n4 = LinearInterpolate(A=@n3.R, B=@n3.R, Alpha=ScalarParameter() {defaultValue=1, parameterName=Rch_or_Ach})
  Multiply(A=ScalarParameter() {defaultValue=1, parameterName=SSS_Thickness_Multiply}, B=@n4)
  ParticleColor()
  n5 = MaterialFunctionCall(Input=ParticleColor().A, Target Low=Constant() {r=1}, Target High=Constant()) {materialFunction=&RemapValueRange}
  Saturate(None=SmoothStep(Min=@n5.Result, Max=@n5.Result, Value=@n4))
  ScalarParameter() {defaultValue=0.1, parameterName=Roughness}
  ScalarParameter() {defaultValue=1, parameterName=Specular}
  ScalarParameter() {parameterName=Metalic}
  TextureSampleParameter2D() {parameterName=NormalMap, samplerType=SAMPLERTYPE_Normal, texture=/Game/Effects/Fort_Effects/Textures/Dirt/T-DirtClodCluster_N}
  VectorParameter() {parameterName=SSS_Color, defaultValue=<rgba 0.4094 0.7312 1 1>}
 ]
}
