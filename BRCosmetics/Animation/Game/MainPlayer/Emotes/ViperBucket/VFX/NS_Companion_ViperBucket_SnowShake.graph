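# Niagara effect for the ViperBucket companion emote that sheds snow as the companion shakes, judging by the asset name.
# Its single emitter moves the spawn point from a head skeletal-mesh sample to a tail sample over a one-second loop, then scatters particles outward under gravity and drag.
# Three spawn groups set the visibility tag for two sprite renderers and a mesh renderer, with separate color attributes for snow smoke and clusters.
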NiagaraSystem /BRCosmetics/Animation/Game/MainPlayer/Emotes/ViperBucket/VFX/NS_Companion_ViperBucket_SnowShake
dependencies {
 usedDataInterfaces [
  &NiagaraDataInterfaceArrayColor
  &NiagaraDataInterfaceArrayFloat
  &NiagaraDataInterfaceCamera
  &NiagaraDataInterfaceCurve
  &NiagaraDataInterfaceSpriteRendererInfo
  &NiagaraDataInterfaceSkeletalMesh
 ]
 usedDynamicInputs [&ReturnCameraVector]
 usedModules [
  &SystemState
  &self:SystemSpawnScript.SystemScriptSource.SystemScriptGraph.NiagaraNodeAssignment_0.SetVariables_#1
  &self:SystemSpawnScript.SystemScriptSource.SystemScriptGraph.NiagaraNodeAssignment_1.SetVariables_#2
 ]
}
emitters {snow_dust_drop {
  emitterUpdateScript [
   &EmitterState(Life Cycle Mode=::NewEnumerator1, Loop Behavior=::NewEnumerator1, Loop Duration=1, UseLoopDelay=false)
   &SpawnRate(SpawnRate=25, Use Spawn Probability=false)
   &SpawnRate(SpawnGroup=1, SpawnRate=50, Use Spawn Probability=false) {module=SpawnRate001}
   &SpawnRate(SpawnGroup=2, SpawnRate=30, Use Spawn Probability=false) {module=SpawnRate002}
  ]
  particleSpawnScript [
   &InitializeParticle(
    Color=&SelectLinearColorFromArray(
     Array Sampling Mode=::NewEnumerator1
     Direct Array Index=&RandomRangeInt(Minimum=0, Maximum=1)
     Color Selection Array=DataInterface() {propertyValues {ColorData [<rgba 0.1812 0.2346 0.3005 1>, <rgba 0.3763 0.4969 0.5395 1>]}}
     Random Seed=?
    )
    Lifetime=&RandomRangeFloat(Minimum=0.4, Maximum=0.5)
    Mesh Scale Mode=::NewEnumerator1
    Mesh Uniform Scale Max=0.02
    Mesh Uniform Scale Min=0.005
    Position Mode=::NewEnumerator0
    Sprite Rotation Mode=::NewEnumerator2
    Sprite Size Mode=::NewEnumerator3
    Sprite UV Mode=::NewEnumerator3
    Uniform Sprite Size=&SelectFloatFromArray(Direct Array Index=&SpawnGroup(), Float Selection Array=DataInterface() {propertyValues {FloatData [30,35,30]}})
    Write Material Random=true
   )
   &SkeletalMeshLocation(
    Bone Sampling Mode=::NewEnumerator4
    MaskSpawnGroup=false
    Position Sampling=::NewEnumerator1
    Sampled Position Offset=<0 0 15>
    Sampled Position Offset Coordinate Space:ENiagaraSkelMeshTransforms=::NewEnumerator0
    Skeletal Mesh=$System.Skeletal Mesh Reader_Head
   )
   &SkeletalMeshLocation(
    Bone Sampling Mode=::NewEnumerator4
    MaskSpawnGroup=false
    Position Sampling=::NewEnumerator1
    Sampled Position Offset Coordinate Space:ENiagaraSkelMeshTransforms=::NewEnumerator0
    Skeletal Mesh=$System.Skeletal Mesh Reader_Tail
   ) {module=SkeletalMeshLocation001}
   /SetVariables(Particles.Position=&Lerp_Vector(
     A=Linked() {linkedVariable {name=Particles.SkeletalMeshLocation.SampledPosition, type=NiagaraPosition}}
     B=Linked() {linkedVariable {name=Particles.SkeletalMeshLocation001.SampledPosition, type=NiagaraPosition}}
     Alpha=$Emitter.NormalizedLoopAge
    )) {isSetParameters=true}
   &ShapeLocation(MaskSpawnGroup=false, Non Uniform Scale=<1 1 0.25>, Offset Mode=::NewEnumerator0, Override Seed=false, Sphere Radius=10, UseOwnerScale=true)
   &AddVelocity(
    Cone Angle=160
    Cone Axis=<0 0 1>
    Module.UseConeFalloff=false
    Module.UseVelDistribution=false
    Override Seed=false
    Rotation Coordinate Space=::World
    Velocity Mode=::NewEnumerator2
    Velocity Speed=&RandomRangeFloat(Minimum=75, Maximum=125)
   )
  ]
  particleUpdateScript [
   &ParticleState()
   /SetVariables(Particles.VisibilityTag=&SpawnGroup()) {isSetParameters=true}
   &ScaleSpriteSize(Uniform Curve Sprite Scale=DataInterface() {propertyValues {Curve {
       defaultValue=0
       keys [
        {arriveTangent=2.4405, time=0, value=0}
        {arriveTangent=1.135, leaveTangent=1.135, time=0.25, value=1}
        {arriveTangent=-1.3333, leaveTangent=-1.3333, time=0.5, value=1}
        {arriveTangent=-3.1685, time=1, value=0}
       ]
      }}})
   &GravityForce(Gravity=&RandomRangeVector(Minimum=<0 0 -550>, Maximum=<0 0 -700>))
   &Drag(Drag=&Lerp_Float(A=1, B=2, Alpha=$Particles.NormalizedAge), Use Linear Drag=true, Use Rotational Drag=false)
   &ScaleColor(
    Scale Alpha=&FloatFromCurve(
     FloatCurve=DataInterface() {propertyValues {Curve {keys [
         <at 0 0 RCIM_Cubic>
         {arriveTangent=0.0808, interpMode=RCIM_Cubic, leaveTangent=0.0808, tangentMode=RCTM_User, time=0.2, value=1}
         {arriveTangent=0.0541, interpMode=RCIM_Cubic, leaveTangent=0.0541, tangentMode=RCTM_User, time=0.6, value=1}
         <at 1 0 RCIM_Cubic>
        ]}}}
     Scale Curve=0.9
    )
    ScaleA=true
    ScaleRGB=false
   )
   &ScaleMeshSize(Scale Factor=&Multiply_VectorByFloat(
     Vector=<1 1 1>
     Float=&FloatFromCurve(FloatCurve=DataInterface() {propertyValues {Curve {keys [
          {arriveTangent=0.6971, time=0, value=0}
          {arriveTangent=-0.0529, leaveTangent=-0.0529, time=0.2, value=1}
          {arriveTangent=-1.6278, leaveTangent=-1.6278, time=0.5, value=1}
          {arriveTangent=-0.0405, leaveTangent=-0.0405, time=1, value=0}
         ]}}})
    ))
   &SubUVAnimation(
    "Sprite Renderer "=DataInterface() {propertyValues {SpriteRenderer=&self:snow_dust_drop_0.NiagaraSpriteRendererProperties_1}}
    Start Frame Range Override=47
    UseEndFrame=true
    UseStartFrame=true
   )
   &SolveForcesAndVelocity(Clamp Velocity=false, Limit Acceleration=false)
   /SetVariables(
    Particles.Color_SnowSmoke:LinearColor=&MakeLinearColorFromVectorAndFloat(
     "Vector (RGB)"=&MakeVectorFromLinearColorRGB(Color=<rgba 0.2028 0.2626 0.3365 1>)
     "Float (Alpha)"=&FloatFromCurve(
      FloatCurve=DataInterface() {propertyValues {Curve {
         defaultValue=1
         keys [
          {arriveTangent=10.4404, interpMode=RCIM_Cubic, leaveTangent=10.4404, tangentMode=RCTM_User, time=0, value=0}
          {arriveTangent=-0.0519, leaveTangent=-0.0519, time=0.15, value=1}
          {arriveTangent=-3.2899, interpMode=RCIM_Cubic, leaveTangent=-3.2899, tangentMode=RCTM_User, time=0.5, value=1}
          {arriveTangent=-0.6101, interpMode=RCIM_Cubic, leaveTangent=-0.6101, tangentMode=RCTM_User, time=1, value=0}
         ]
        }}}
      Scale Curve=0.42
     )
    )
    Particles.SpriteSize_SnowSmoke:Vector2f=&Lerp_Vector2(A=<30 30>, B=<75 75>, Alpha=$Particles.NormalizedAge)
   ) {isSetParameters=true}
   &CameraOffset(Camera Offset Amount=25)
   /SetVariables(Particles.Color_SnowCluster:LinearColor=&MakeLinearColorFromVectorAndFloat(
     "Vector (RGB)"=&MakeVectorFromLinearColorRGB(Color=<rgba 0.5774 0.6317 0.699 1>)
     "Float (Alpha)"=&FloatFromCurve(
      FloatCurve=DataInterface() {propertyValues {Curve {
         defaultValue=1
         keys [
          {arriveTangent=10.4404, time=0, value=0}
          {arriveTangent=-0.0519, leaveTangent=-0.0519, time=0.1, value=1}
          {arriveTangent=-3.2899, leaveTangent=-3.2899, time=0.75, value=1}
          {arriveTangent=-0.6101, leaveTangent=-0.6101, time=1, value=0}
         ]
        }}}
      Scale Curve=0.41
     )
    )) {isSetParameters=true}
  ]
  rendererClasses [&NiagaraSpriteRendererProperties, &NiagaraMeshRendererProperties]
  renderers [
   {rendererClass=&NiagaraSpriteRendererProperties}
   {rendererClass=&NiagaraSpriteRendererProperties, rendererIndex=1}
   {rendererClass=&NiagaraMeshRendererProperties, rendererIndex=2}
  ]
 }}
summary {
 systemName=NS_Companion_ViperBucket_SnowShake
 emitters [{emitterName=snow_dust_drop, rendererClasses [&NiagaraSpriteRendererProperties, &NiagaraMeshRendererProperties]}]
 userVariables [
  User.bIsFrontEnd {type=NiagaraBool}
  User.bIsFrontEndPreview {type=NiagaraBool}
 ]
}
