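# Niagara system for the ViperBucket emote that renders a blizzard of wind-driven snow, smoke, fog and air streaks.
# BlizzardAir, SmokeGround001 and SnowDust001 spawn sprites pushed by wind, gravity or curl noise, with snowflake alpha driven by facing toward the camera.
# Fog is a long-lived mesh and sprite haze, while Air_Trail002 draws looping ribbon streaks from staggered bursts.
# System-level camera-fade values scale sizes and alphas, and time-of-day values dim some layers.
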NiagaraSystem /BRCosmetics/Animation/Game/MainPlayer/Emotes/ViperBucket/VFX/NS_Emote_ViperBucket_Blizzard
dependencies {
 usedDataInterfaces [
  &NiagaraDataInterfaceArrayFloat
  &NiagaraDataInterfaceArrayColor
  &NiagaraDataInterfaceSpriteRendererInfo
  &NiagaraDataInterfaceCurve
  &NiagaraDataInterfaceArrayFloat3
  &NiagaraDataInterfaceCamera
  &NiagaraDataInterfaceColorCurve
 ]
 usedDynamicInputs [&SimulationPosition, null, /Niagara/DynamicInputs/Camera/CameraFade]
 usedModules [
  &self:SystemSpawnScript.SystemScriptSource.SystemScriptGraph.NiagaraNodeAssignment_0.SetVariables_#1
  &SystemState
  &self:SystemSpawnScript.SystemScriptSource.SystemScriptGraph.NiagaraNodeAssignment_1.SetVariables_#2
 ]
}
emitters {
 Air_Trail002 {
  emitterUpdateScript [
   &EmitterState(
    Delay First Loop Only=true
    Life Cycle Mode=::NewEnumerator1
    Loop Behavior=::NewEnumerator2
    Loop Count=5
    Loop Delay=0.2
    Loop Duration=1.55
    UseLoopDelay=true
   )
   &SpawnBurst_Instantaneous(Spawn Count=3, Use Spawn Probability=false, UseLoopCountLimit=false)
   &SpawnBurst_Instantaneous(Spawn Count=3, Spawn Group=1, Spawn Time=0.2, Use Spawn Probability=false, UseLoopCountLimit=false) {module=SpawnBurst_Instantaneous001}
   &SpawnBurst_Instantaneous(Spawn Count=5, Spawn Group=2, Spawn Time=0.4, Use Spawn Probability=false, UseLoopCountLimit=false) {module=SpawnBurst_Instantaneous002}
   &SpawnBurst_Instantaneous(Spawn Count=3, Spawn Group=3, Spawn Time=0.7, Use Spawn Probability=false, UseLoopCountLimit=false) {module=SpawnBurst_Instantaneous003}
   &SpawnBurst_Instantaneous(Spawn Count=3, Spawn Group=4, Spawn Time=0.8, Use Spawn Probability=false, UseLoopCountLimit=false) {module=SpawnBurst_Instantaneous004}
   &SpawnBurst_Instantaneous(Spawn Count=3, Spawn Group=5, Spawn Time=1, Use Spawn Probability=false, UseLoopCountLimit=false) {module=SpawnBurst_Instantaneous005}
  ]
  particleSpawnScript [
   &InitializeParticle(
    Color=&MultiplyLinearColor(
     Linear Color=<rgba 0.6867 0.9047 1 1>
     Scale Factor=&MakeFloat4FromVectorAndFloat(
      "Vector (RGB)"=<1 1 1>
      "Float (Alpha)"=&FloatFromCurve(
       FloatCurve=DataInterface() {propertyValues {Curve {keys [
           {arriveTangent=5.3189, time=0, value=0}
           {arriveTangent=2, leaveTangent=2, time=0.05, value=1}
           <at 0.5 1>
           {arriveTangent=-5.94, time=1, value=0}
          ]}}}
       CurveIndex=&ReturnNormalizedExecIndex()
       Scale Curve=&RandomRangeFloat(Minimum=0.6, Maximum=0.9)
      )
     )
    )
    Lifetime=&RandomRangeFloat(Minimum=0.75, Maximum=1)
    Position Mode=::NewEnumerator0
    Ribbon Width=&FloatFromCurve(
     FloatCurve=DataInterface() {propertyValues {Curve {keys [<at 0 1>,<at 1 1.75>]}}}
     CurveIndex=&ReturnNormalizedExecIndex()
     Scale Curve=30
    )
    Ribbon Width Mode=::NewEnumerator1
    Write Material Random=true
   )
   /SetVariables(Particles.Position=&AddVectorToPosition(
     Position=&SelectVectorFromArray(
      Direct Array Index=&SpawnGroup()
      Vector Selection Array=DataInterface() {propertyValues {FloatData [<49.4851 -30 117>, <46.5605 20 100>, <15.8955 15 83>, <28.7763 30 77>, <65.8782 -20 65>, <-10 -15 85>]}}
     )
     Vector=&MakeVector(
      X=&Lerp_Float(A=190, B=-130, Alpha=&ReturnNormalizedExecIndex())
      Y=0
      Z=&Add_Float(
       A=&FloatFromCurve(
        FloatCurve=DataInterface() {propertyValues {Curve {keys [
            {arriveTangent=-7.0753, interpMode=RCIM_Cubic, leaveTangent=-7.0753, tangentMode=RCTM_User, time=0, value=2}
            {arriveTangent=-0.9398, interpMode=RCIM_Cubic, leaveTangent=-0.9398, tangentMode=RCTM_User, time=1, value=-1.5}
           ]}}}
        CurveIndex=&ReturnNormalizedExecIndex()
        Scale Curve=50
       )
       B=50
      )
     )
    )) {isSetParameters=true}
   /SetVariables(Particles.RibbonLinkOrder:NiagaraFloat=&ReturnNormalizedExecIndex()) {isSetParameters=true}
   /SetVariables(Particles.VisibilityTag=&SpawnGroup(Max Spawn Group=5)) {isSetParameters=true}
  ]
  particleUpdateScript [
   &ParticleState()
   &ScaleColor(
    Scale Alpha=&FloatFromCurve(
     FloatCurve=DataInterface() {propertyValues {Curve {keys [{arriveTangent=0.2265, time=0, value=0}, <at 0.4 1>, <at 0.6 1>, {tangentMode=RCTM_User, time=1, value=0}]}}}
     CurveIndex=&ClampFloat(Float=&Add_Float(
       A=&Multiply_Float(A=&ReturnNormalizedExecIndex(), B=1)
       B=&FloatFromCurve(FloatCurve=DataInterface() {propertyValues {Curve {keys [{arriveTangent=-1.5875, time=0, value=1}, {arriveTangent=-2.3125, leaveTangent=-2.3125, time=1, value=-0.5}]}}})
      ))
     Scale Curve=&FloatFromCurve(
      FloatCurve=DataInterface() {propertyValues {Curve {keys [<at 0 0>,<at 0.4 1>,<at 0.6 1>,<at 1 0>]}}}
      Scale Curve=&SelectFloatFromArray(
       Direct Array Index=&SpawnGroup()
       Float Selection Array=DataInterface() {propertyValues {FloatData [0.2,0.7,0.7,0.4,0.7,0.5]}}
      )
     )
    )
    ScaleA=true
    ScaleRGB=false
   )
   &ScaleRibbonWidth(Ribbon Width Scale=&SelectFloatFromArray(
     Direct Array Index=&SpawnGroup()
     Float Selection Array=DataInterface() {propertyValues {FloatData [0.2,0.1,0.15,0.25,0.15,0.12]}}
    ))
  ]
  rendererClasses [&NiagaraRibbonRendererProperties]
  renderers [
   {rendererClass=&NiagaraRibbonRendererProperties}
   {rendererClass=&NiagaraRibbonRendererProperties, rendererIndex=1}
   {rendererClass=&NiagaraRibbonRendererProperties, rendererIndex=2}
   {rendererClass=&NiagaraRibbonRendererProperties, rendererIndex=3}
   {rendererClass=&NiagaraRibbonRendererProperties, rendererIndex=4}
   {rendererClass=&NiagaraRibbonRendererProperties, rendererIndex=5}
  ]
 }
 BlizzardAir {
  emitterUpdateScript [
   &EmitterState(Life Cycle Mode=::NewEnumerator1, Loop Behavior=::NewEnumerator1, Loop Delay=1, Loop Duration=6, UseLoopDelay=true)
   &SpawnRate(Spawn Probability=0.75, SpawnRate=40, Use Spawn Probability=false)
   &SpawnRate(Spawn Probability=0.5, SpawnGroup=1, SpawnRate=30, Use Spawn Probability=true) {module=SpawnRate002}
   &SpawnRate(Spawn Probability=0.75, SpawnGroup=2, SpawnRate=45, Use Spawn Probability=true) {module=SpawnRate001}
  ]
  particleSpawnScript [
   &InitializeParticle(
    Color Mode=::NewEnumerator0
    Lifetime Max=&SelectFloatFromArray(
     Direct Array Index=&SpawnGroup()
     Float Selection Array=DataInterface() {propertyValues {FloatData [1,1.25,1,0.75]}}
    )
    Lifetime Min:NiagaraFloat=&SelectFloatFromArray(
     Direct Array Index=&SpawnGroup()
     Float Selection Array=DataInterface() {propertyValues {FloatData [0.75,0.625,0.75,0.375]}}
    )
    Lifetime Mode=::NewEnumerator1
    Mesh Scale Mode=::NewEnumerator1
    Mesh Uniform Scale Max=0.02
    Mesh Uniform Scale Min=0.01
    Position Offset=<160 0 200>
    Sprite Rotation Mode=::NewEnumerator2
    Sprite Size Mode=::NewEnumerator1
    Uniform Sprite Size Max=225
    Uniform Sprite Size Min=175
    UsePositionOffset=true
    Write Material Random=true
   )
   &ShapeLocation(
    MaskSpawnGroup=false
    Offset Mode=::NewEnumerator0
    Override Seed=false
    Sphere Radius=&SelectFloatFromArray(
     Direct Array Index=&SpawnGroup()
     Float Selection Array=DataInterface() {propertyValues {FloatData [75,75,100]}}
    )
    UseOwnerScale=true
   ) {module=ShapeLocation001}
   &SubUVAnimation(
    Direct Frame Index:NiagaraFloat=/Niagara/DynamicInputs/Math/Floor_Float(Float:NiagaraFloat=&RandomRangeFloat(Minimum=0, Maximum=3.99))
    "Sprite Renderer "=DataInterface() {propertyValues {SpriteRenderer=&self:BlizzardAir_0.Renderer}}
    SubUV Animation Mode:ENiagara_SubUVLookupModeV2=::NewEnumerator5
   ) {module=SubUVAnimation001}
  ]
  particleUpdateScript [
   &ParticleState()
   /SetVariables(
    Particles.SpriteAlignment_BlizzardVel:Vector3f=&AddVectorToPosition(Position=$Particles.Position, Vector=<-1 0 0>)
    Particles.SpriteFacing_SnowFlakes:Vector3f=&Lerp_Vector(
     A=&RandomRangeVector(Minimum=<-1 -1 -1>, Maximum=<1 1 1>)
     B=&RandomRangeVector(Minimum=<-1 -1 -1>, Maximum=<1 1 1>)
     Alpha=&Sine(Normalized Angle=$Emitter.Age, Period=1, Scale=1, Bias=0)
    )
   ) {isSetParameters=true}
   &WindForce(
    Turbulence Mode:ENiagaraWindTurbulenceMode=::NewEnumerator0
    Wind Speed=<-1 0 0>
    Wind Speed Scale=&Multiply_Float(
     A=&Lerp_Float(
      A=2000
      B=2750
      Alpha=&FloatFromCurve(FloatCurve=DataInterface() {propertyValues {Curve {
          defaultValue=0
          keys [
           {arriveTangent=0.3662, interpMode=RCIM_Cubic, leaveTangent=0.3662, tangentMode=RCTM_User, time=0, value=0}
           {arriveTangent=2.3876, interpMode=RCIM_Cubic, leaveTangent=2.3876, tangentMode=RCTM_User, time=1, value=1}
          ]
         }}})
     )
     B=&RandomRangeFloat(Minimum=1, Maximum=1)
    )
   )
   &ScaleSpriteSize(Scale Factor=&Multiply_Vector2D_ByFloat(
     Float=&FloatFromCurve(
      FloatCurve=DataInterface() {propertyValues {Curve {keys [
          {arriveTangent=0.1908, interpMode=RCIM_Cubic, leaveTangent=0.1908, tangentMode=RCTM_User, time=0, value=0.5}
          {arriveTangent=0.7057, interpMode=RCIM_Cubic, leaveTangent=0.7057, tangentMode=RCTM_User, time=0.8, value=0.8}
          {arriveTangent=1.1151, interpMode=RCIM_Cubic, leaveTangent=1.1151, tangentMode=RCTM_User, time=1, value=1.5}
         ]}}}
      Scale Curve=$System.Size_CameraFade_Near
     )
     Vector2D=<0.9 1.8>
    ))
   &SubUVAnimation(
    Random Start Frame=true
    "Sprite Renderer "=DataInterface() {propertyValues {SpriteRenderer=&self:BlizzardAir_0.NiagaraSpriteRendererProperties_1}}
    UseEndFrame=false
    UseStartFrame=false
   )
   &DynamicMaterialParameters(
    Index 0 Param 1=1
    Index 0 Param 4=1
    Module.Index0Float0Write=true
    Module.Index0Float1Write=false
    Module.Index0Float2Write=false
    Module.Index0Float3Write=true
   )
   &GravityForce(Gravity=&Lerp_Vector(
     A=<0 0 -1750>
     B=&SelectVectorFromArray(
      Direct Array Index=&SpawnGroup()
      Vector Selection Array=DataInterface() {propertyValues {FloatData [<0 0 2500>, <0 0 2500>, <0 0 2000>, <0 0 1000>]}}
     )
     Alpha=$Particles.NormalizedAge
    )) {module=GravityForce001}
   &Drag(Drag=0.5, Use Linear Drag=true, Use Rotational Drag=false)
   &CurlNoiseForce(
    Bias Noise=false
    Noise Frequency=25
    Noise Strength=&SelectFloatFromArray(
     Direct Array Index=&SpawnGroup()
     Float Selection Array=DataInterface() {propertyValues {FloatData [500,500,1000,500]}}
    )
    Pan Noise=false
   )
   &SolveForcesAndVelocity(Clamp Velocity=false, Limit Acceleration=false)
   /SetVariables(Particles.VisibilityTag=&SpawnGroup(Max Spawn Group=2)) {isSetParameters=true, module=SetVariables_#3}
   /SetVariables(
    Particles.Color_BlizzardSmoke:LinearColor=&MakeLinearColorFromVectorAndFloat(
     "Vector (RGB)"=&MakeVectorFromLinearColorRGB(Color=<rgba 0.2502 0.3968 0.5029 1>)
     "Float (Alpha)"=&FloatFromCurve(
      FloatCurve=DataInterface() {propertyValues {Curve {keys [
          {interpMode=RCIM_Cubic, tangentMode=RCTM_User, time=0, value=0}
          {arriveTangent=1.6667, interpMode=RCIM_Cubic, leaveTangent=1.6667, time=0.3274, value=0.2272}
          {arriveTangent=0.1753, interpMode=RCIM_Cubic, leaveTangent=0.1753, tangentMode=RCTM_User, time=0.6, value=1}
          {arriveTangent=-2.5, interpMode=RCIM_Cubic, leaveTangent=-2.5, time=0.868, value=0.121}
          {arriveTangent=0.0016, tangentMode=RCTM_User, time=1, value=0}
         ]}}}
      Scale Curve=&Multiply_Float(A=&RandomRangeFloat(Minimum=0.3, Maximum=0.6), B=1)
     )
    )
    Particles.Color_SnowCluster:LinearColor=&MakeLinearColorFromVectorAndFloat(
     "Vector (RGB)"=<1 1 1>
     "Float (Alpha)"=&FloatFromCurve(FloatCurve=DataInterface() {propertyValues {Curve {keys [<at 0 0>,<at 0.3 1>,<at 0.8 1>,<at 1 0>]}}}, Scale Curve=0.4)
    )
    Particles.Color_SnowFlakes:LinearColor=&MakeLinearColorFromVectorAndFloat(
     "Vector (RGB)"=&MakeVectorFromLinearColorRGB(Color=&SelectLinearColorFromArray(
       Array Sampling Mode=::NewEnumerator1
       Direct Array Index=&RandomRangeInt(Minimum=0, Maximum=2)
       Color Selection Array=DataInterface() {propertyValues {ColorData [<rgba 0.34 0.5775 1 1>, <rgba 0.9 0.9 0.9 1>, <rgba 0.34 0.8341 1 1>]}}
       Random Seed=?
      ))
     "Float (Alpha)"=&ClampFloat(Float=&RemapRange(
       Input Value To Remap=&Abs_Float(Float:NiagaraFloat=&DotProduct(
         Vector A=$Particles.SpriteFacing_SnowFlakes
         Vector B=&ReturnCameraVector(Camera Vectors=::NewEnumerator1, Coordinate Space=::Simulation)
        ))
       Input Min=-1
       Input Max=1
       Output Min=-1.5
       Output Max=1
      ))
    )
   ) {isSetParameters=true}
   /SetVariables(
    Particles.SpriteSize_SnowCluster:Vector2f=<50 300>
    Particles.SpriteSize_SnowFlakes:Vector2f=&RandomRangeVector2D(Minimum=<5 5>, Maximum=<10 10>)
   ) {isSetParameters=true, module=SetVariables_#4}
  ]
  rendererClasses [&NiagaraSpriteRendererProperties]
  renderers [
   {rendererClass=&NiagaraSpriteRendererProperties}
   {rendererClass=&NiagaraSpriteRendererProperties, rendererIndex=1}
   {rendererClass=&NiagaraSpriteRendererProperties, rendererIndex=2}
  ]
 }
 Fog {
  emitterUpdateScript [
   &EmitterState(Inactive Response=::NewEnumerator1, Life Cycle Mode=::NewEnumerator1, Loop Behavior=::NewEnumerator1, Loop Duration=8.5, UseLoopDelay=false)
   &SpawnBurst_Instantaneous(Use Spawn Probability=false, UseLoopCountLimit=false)
  ]
  particleSpawnScript [
   &InitializeParticle(
    Color Mode=::NewEnumerator0
    Lifetime=8.5
    Mesh Scale=<5.4 2.025 4.05>
    Mesh Scale Mode=::NewEnumerator4
    Position Offset=<20 0 100>
    Sprite Rotation Mode=::NewEnumerator0
    Sprite Size Mode=::NewEnumerator3
    Uniform Sprite Size=700
    UsePositionOffset=true
    Write Material Random=true
   )
   /SetVariables(Particles.Position_Floor:NiagaraPosition=&AddVectorToPosition(Position=$Engine.Emitter.SimulationPosition, Vector=<0 0 5>)) {isSetParameters=true}
   /SetVariables(Particles.SpriteSize_UP:Vector2f=<500 300>) {isSetParameters=true}
  ]
  particleUpdateScript [
   &ParticleState() {module=ParticleState001}
   &ScaleMeshSize(Scale Factor=&Multiply_VectorByFloat(Vector=<1 1 1>, Float=$System.Size_CameraFade_Near))
   &DynamicMaterialParameters(
    Index 0 Param 1=100
    Index 0 Param 2=1
    Index 0 Param 3=&FloatFromCurve(FloatCurve=DataInterface() {propertyValues {Curve {defaultValue=1, keys [<at 0.1 1>,<at 0.2 0>,<at 0.85 0>,<at 1 1>]}}})
    Module.Index0Float0Write=true
    Module.Index0Float1Write=true
    Module.Index0Float2Write=true
    Module.Index0Float3Write=false
   )
   /SetVariables(Particles.SpriteSize_Halo:Vector2f=&Multiply_Vector2D_ByFloat(Float=$System.Size_CameraFade_Near, Vector2D=<700 700>)) {isSetParameters=true}
   /SetVariables(
    Particles.Color_Halo:LinearColor=&MakeLinearColorFromVectorAndFloat(
     "Vector (RGB)"=&MakeVectorFromLinearColorRGB(Color=<rgba 0.402 0.6867 1 1>)
     "Float (Alpha)"=&FloatFromCurve(
      FloatCurve=DataInterface() {propertyValues {Curve {keys [<at 0 0>,<at 0.1 1>,<at 0.95 1>,<at 1 0>]}}}
      Scale Curve=&Multiply_Float(A=0.05, B=$System.Alpha_CameraFade_Near)
     )
    )
    Particles.Color_MeshSphere:LinearColor=&MakeLinearColorFromVectorAndFloat(
     "Vector (RGB)"=&MakeVectorFromLinearColorRGB(Color=<rgba 2.0099 3.4334 5 1>)
     "Float (Alpha)"=&FloatFromCurve(
      FloatCurve=DataInterface() {propertyValues {Curve {keys [<at 0 0>,<at 0.1 1>,<at 0.95 1>,<at 1 0>]}}}
      Scale Curve=&Multiply_Float(A=0.6, B=$System.Alpha_CameraFade_Near)
     )
    )
    Particles.Color_SpriteUp:LinearColor=&MakeLinearColorFromVectorAndFloat(
     "Vector (RGB)"=&MakeVectorFromLinearColorRGB(Color=<rgba 0.402 0.6867 1 1>)
     "Float (Alpha)"=&FloatFromCurve(
      FloatCurve=DataInterface() {propertyValues {Curve {keys [<at 0 0>,<at 0.1 1>,<at 0.95 1>,<at 1 0>]}}}
      Scale Curve=&Multiply_Float(A=0.4, B=$System.FN_TimeOfDay)
     )
    )
   ) {isSetParameters=true}
   /SetVariables(Particles.SpriteAlignment_WindDirection:Vector3f=<0 -1 0>, Particles.SpriteFacing_Up:Vector3f=<0 0 1>) {isSetParameters=true}
  ]
  rendererClasses [&NiagaraSpriteRendererProperties, &NiagaraMeshRendererProperties]
  renderers [
   {rendererClass=&NiagaraSpriteRendererProperties}
   {rendererClass=&NiagaraSpriteRendererProperties, rendererIndex=1}
   {rendererClass=&NiagaraMeshRendererProperties, rendererIndex=2}
  ]
 }
 SmokeGround001 {
  emitterUpdateScript [
   &EmitterState(Life Cycle Mode=::NewEnumerator1, Loop Behavior=::NewEnumerator1, Loop Delay=1.6, Loop Duration=5.9, UseLoopDelay=true)
   &SpawnRate(Spawn Probability=0.75, SpawnRate=30, Use Spawn Probability=true)
   &SpawnRate(Spawn Probability=0.75, SpawnGroup=1, SpawnRate=30, Use Spawn Probability=true) {module=SpawnRate004}
   &SpawnRate(Spawn Probability=0.75, SpawnGroup=2, SpawnRate=10, Use Spawn Probability=true) {module=SpawnRate001}
   &SpawnRate(Spawn Probability=0.5, SpawnGroup=3, SpawnRate=40, Use Spawn Probability=true) {module=SpawnRate002}
   &SpawnRate(Spawn Probability=0.25, SpawnGroup=4, SpawnRate=15, Use Spawn Probability=true) {module=SpawnRate003}
  ]
  particleSpawnScript [
   &InitializeParticle(
    Color=&SelectLinearColorFromArray(
     Array Sampling Mode=::NewEnumerator1
     Direct Array Index=&SpawnGroup()
     Color Selection Array=DataInterface() {propertyValues {ColorData [<rgba 0.2423 0.3663 0.4564 0.5>, <rgba 0.2423 0.3663 0.4564 0.5>, <rgba 0 0.9131 1 1>, <rgba 1 1 1 1>]}}
     Random Seed=?
    )
    Lifetime Max=&SelectFloatFromArray(
     Direct Array Index=&SpawnGroup()
     Float Selection Array=DataInterface() {propertyValues {FloatData [0.6,0.6,0.9,0.5]}}
    )
    Lifetime Min:NiagaraFloat=&SelectFloatFromArray(
     Direct Array Index=&SpawnGroup()
     Float Selection Array=DataInterface() {propertyValues {FloatData [0.375,0.375,0.4688,0.2812]}}
    )
    Lifetime Mode=::NewEnumerator1
    Mesh Scale Mode=::NewEnumerator1
    Mesh Uniform Scale Max=0.02
    Mesh Uniform Scale Min=0.01
    Sprite Rotation Mode=::NewEnumerator2
    Sprite Size Mode=::NewEnumerator1
    Uniform Sprite Size Max=110
    Uniform Sprite Size Min=80
    UsePositionOffset=true
    Write Material Random=true
   )
   &ShapeLocation(
    Box Size=<100 100 0>
    MaskSpawnGroup=false
    Offset Mode=::NewEnumerator0
    Override Seed=false
    Shape Primitive=::NewEnumerator2
    Surface Only=false
    UseOwnerScale=true
   ) {module=ShapeLocation001}
   /SetVariables(Particles.SpriteFacing_SnowFlakes:Vector3f=&Lerp_Vector(
     A=&RandomRangeVector(Minimum=<-1 -1 -1>, Maximum=<1 1 1>)
     B=&RandomRangeVector(Minimum=<-1 -1 -1>, Maximum=<1 1 1>)
     Alpha=&Sine(Normalized Angle=$Emitter.Age, Period=1, Scale=1, Bias=0)
    )) {isSetParameters=true}
   &SubUVAnimation(
    Direct Frame Index:NiagaraFloat=&Floor_Float(Float:NiagaraFloat=&RandomRangeFloat(Minimum=0, Maximum=3.99))
    "Sprite Renderer "=DataInterface() {propertyValues {SpriteRenderer=&self:SmokeGround001_0.Renderer}}
    SubUV Animation Mode:ENiagara_SubUVLookupModeV2=::NewEnumerator5
   ) {module=SubUVAnimation003}
  ]
  particleUpdateScript [
   &ParticleState()
   &WindForce(Frequency=110, Scale=1, Turbulence Offset Mode:ENiagaraWindOffsetMode=::NewEnumerator0, Wind Speed=<-1 0 0>, Wind Speed Scale=1600)
   &ScaleColor(
    Scale Alpha=&FloatFromCurve(FloatCurve=DataInterface() {propertyValues {Curve {keys [
         {interpMode=RCIM_Cubic, tangentMode=RCTM_User, time=0, value=0}
         {interpMode=RCIM_Cubic, tangentMode=RCTM_User, time=0.5, value=1}
         {tangentMode=RCTM_User, time=1, value=0}
        ]}}})
    Scale RGB=&MakeVectorFromLinearColorRGB(Color=<rgba 0.7677 0.9954 1 1>)
    ScaleA=true
    ScaleRGB=false
   )
   &SpriteRotationRate(Rotation Rate=&RandomRangeFloat(Minimum=-60, Maximum=60))
   &ScaleSpriteSize(Scale Factor=&Vector2DFromFloat(Value=&FloatFromCurve(
      FloatCurve=DataInterface() {propertyValues {Curve {keys [
          {arriveTangent=1.2135, interpMode=RCIM_Cubic, leaveTangent=1.2135, tangentMode=RCTM_User, time=0, value=0.5}
          {arriveTangent=0.6636, interpMode=RCIM_Cubic, leaveTangent=0.6636, tangentMode=RCTM_User, time=0.3509, value=0.7982}
          {arriveTangent=0.0895, interpMode=RCIM_Cubic, leaveTangent=0.0895, tangentMode=RCTM_User, time=1, value=1}
         ]}}}
      Scale Curve=$System.Size_CameraFade_Near
     )))
   &SubUVAnimation(
    Random Start Frame=true
    "Sprite Renderer "=DataInterface() {propertyValues {SpriteRenderer=&self:SmokeGround001_0.NiagaraSpriteRendererProperties_2}}
    UseEndFrame=false
    UseStartFrame=false
   )
   &SubUVAnimation(
    Random Start Frame=true
    "Sprite Renderer "=DataInterface() {propertyValues {SpriteRenderer=&self:SmokeGround001_0.NiagaraSpriteRendererProperties_1}}
    UseEndFrame=false
    UseStartFrame=false
   ) {module=SubUVAnimation002}
   &DynamicMaterialParameters(
    Index 0 Param 1=1
    Index 0 Param 4=1
    Module.Index0Float0Write=true
    Module.Index0Float1Write=false
    Module.Index0Float2Write=false
    Module.Index0Float3Write=true
   )
   &CurlNoiseForce(
    Bias Noise=false
    Noise Frequency=25
    Noise Strength=&SelectFloatFromArray(
     Direct Array Index=&SpawnGroup()
     Float Selection Array=DataInterface() {propertyValues {FloatData [0,0,1000,0]}}
    )
    Pan Noise=false
   )
   &Drag(Use Linear Drag=true, Use Rotational Drag=false)
   &GravityForce(Gravity=<0 0 100>)
   &SolveForcesAndVelocity(Clamp Velocity=false, Limit Acceleration=false)
   &CameraOffset(Camera Offset Amount=50)
   /SetVariables(Particles.VisibilityTag=&SpawnGroup(Max Spawn Group=5)) {isSetParameters=true, module=SetVariables_#3}
   /SetVariables(
    Particles.Color_SnowClusterErode:LinearColor=&LerpLinearColors(
     StartColor=<rgba 1 1 1 0.7>
     EndColor=<rgba 1 1 1 0.5>
     LerpFactor=&FloatFromCurve(FloatCurve=DataInterface() {propertyValues {Curve {keys [
          {arriveTangent=7.7784, interpMode=RCIM_Cubic, leaveTangent=7.7784, tangentMode=RCTM_User, time=0, value=0}
          {arriveTangent=-0.0278, interpMode=RCIM_Cubic, leaveTangent=-0.0278, tangentMode=RCTM_User, time=0.4, value=1}
          {arriveTangent=-0.1012, interpMode=RCIM_Cubic, leaveTangent=-0.1012, tangentMode=RCTM_User, time=0.6, value=1}
          {arriveTangent=-7.0772, interpMode=RCIM_Cubic, leaveTangent=-7.0772, tangentMode=RCTM_User, time=1, value=0}
         ]}}})
    )
    Particles.Color_SnowFlakes:LinearColor=&MakeLinearColorFromVectorAndFloat(
     "Vector (RGB)"=&MakeVectorFromLinearColorRGB(Color=&SelectLinearColorFromArray(
       Array Sampling Mode=::NewEnumerator1
       Direct Array Index=&RandomRangeInt(Minimum=0, Maximum=2)
       Color Selection Array=DataInterface() {propertyValues {ColorData [<rgba 0.34 0.5775 1 1>, <rgba 0.9 0.9 0.9 1>, <rgba 0.34 0.8341 1 1>]}}
       Random Seed=?
      ))
     "Float (Alpha)"=&ClampFloat(Float=&RemapRange(
       Input Value To Remap=&Abs_Float(Float:NiagaraFloat=&DotProduct(
         Vector A=$Particles.SpriteFacing_SnowFlakes
         Vector B=&ReturnCameraVector(Camera Vectors=::NewEnumerator1, Coordinate Space=::Simulation)
        ))
       Input Min=-1
       Input Max=1
       Output Min=-1.5
       Output Max=1
      ))
    )
   ) {isSetParameters=true}
   /SetVariables(
    Particles.SpriteSize_Blizzard_VelAlign:Vector2f=&Multiply_Vector2D_ByFloat(
     Float=&FloatFromCurve(
      FloatCurve=DataInterface() {propertyValues {Curve {keys [
          {arriveTangent=0.1908, interpMode=RCIM_Cubic, leaveTangent=0.1908, tangentMode=RCTM_User, time=0, value=0.5}
          {arriveTangent=0.0122, interpMode=RCIM_Cubic, leaveTangent=0.0122, tangentMode=RCTM_User, time=1, value=1}
         ]}}}
      Scale Curve=$System.Size_CameraFade_Near
     )
     Vector2D=<125 225>
    )
    Particles.SpriteSize_SnowCluster:Vector2f=&Lerp_Vector2(
     A=<0 0>
     B=<300 300>
     Alpha=&FloatFromCurve(FloatCurve=DataInterface() {propertyValues {Curve {keys [<at 0 0>,<at 0.4 1>,<at 0.6 1>,<at 1 0>]}}})
    )
    Particles.SpriteSize_SnowFlakes:Vector2f=&RandomRangeVector2D(Minimum=<5 5>, Maximum=<10 10>)
   ) {isSetParameters=true, module=SetVariables_#4}
   /SetVariables() {isSetParameters=true}
  ]
  rendererClasses [&NiagaraSpriteRendererProperties]
  renderers [
   {rendererClass=&NiagaraSpriteRendererProperties}
   {rendererClass=&NiagaraSpriteRendererProperties, rendererIndex=1}
   {rendererClass=&NiagaraSpriteRendererProperties, rendererIndex=2}
  ]
 }
 SnowDust001 {
  emitterUpdateScript [
   &EmitterState(Life Cycle Mode=::NewEnumerator1, Loop Behavior=::NewEnumerator1, Loop Delay=0.5, Loop Duration=7.5, UseLoopDelay=true)
   &SpawnRate(SpawnRate=50, Use Spawn Probability=false)
  ]
  particleSpawnScript [
   &InitializeParticle(
    Color=<rgba 0.3419 0.7011 0.9473 1>
    Lifetime=&RandomRangeFloat(Minimum=0.5, Maximum=0.75)
    Position Offset=<75 0 50>
    Sprite Rotation Angle=45
    Sprite Rotation Mode=::NewEnumerator0
    Sprite Size Mode=::NewEnumerator1
    Uniform Sprite Size Max=150
    Uniform Sprite Size Min=100
    UsePositionOffset=true
    Write Material Random=true
   )
   &ShapeLocation(
    Box Midpoint=<0.5 0.5 0>
    Box Size=<275 250 200>
    MaskSpawnGroup=false
    Override Seed=false
    Shape Primitive=::NewEnumerator2
    Surface Only=false
    UseOwnerScale=true
   )
  ]
  particleUpdateScript [
   &ParticleState()
   &ScaleColor(
    Scale Alpha=&FloatFromCurve(
     FloatCurve=DataInterface() {propertyValues {Curve {keys [<at 0 0>,<at 0.5 1>,<at 1 0>]}}}
     Scale Curve=&Multiply_Float(
      A=&RandomRangeFloat(Minimum=0.35, Maximum=0.7)
      B=&Multiply_Float(
       A=&ClampFloat(Float=&RemapRange(Input Value To Remap=$Emitter.Age, Input Min=7, Input Max=8, Output Min=1, Output Max=0))
       B=$System.Alpha_CameraFade_Near
      )
     )
    )
    ScaleA=true
    ScaleRGB=false
   )
   &ScaleSpriteSize(
    Uniform Curve Scale=$System.Size_CameraFade_Near
    Uniform Curve Sprite Scale=DataInterface() {propertyValues {Curve {keys [{arriveTangent=3.5003, time=0, value=0.5}, <at 1.2 1.2>]}}}
   )
   &GravityForce(Coordinate Space=::Local, Gravity=<-500 0 -300>)
   &DynamicMaterialParameters(
    Index 0 Param 1=&FloatFromCurve(FloatCurve=DataInterface() {propertyValues {Curve {keys [<at 0 0 RCIM_Cubic>,<at 1.0011 0.2>]}}})
    Index 0 Param 3=1
    Module.Index0Param0Write=false
    Module.Index0Param1Write=false
    Module.Index0Param2Write=true
    Module.Index0Param3Write=false
   )
   &ScaleColor(
    Scale Alpha=&FNTimeOfDayScalar(Night Scale=0.6, Morning Scale=1, Evening Scale=0.8)
    ScaleA=true
    ScaleRGB=true
   ) {module=ScaleColor002}
   &SolveForcesAndVelocity(Clamp Velocity=false, Limit Acceleration=false)
  ]
  rendererClasses [&NiagaraSpriteRendererProperties]
  renderers [{rendererClass=&NiagaraSpriteRendererProperties}]
 }
}
summary {
 systemName=NS_Emote_ViperBucket_Blizzard
 emitters [
  {emitterName=BlizzardAir, rendererClasses [&NiagaraSpriteRendererProperties]}
  {emitterName=SmokeGround001, rendererClasses [&NiagaraSpriteRendererProperties]}
  {emitterName=Air_Trail002, rendererClasses [&NiagaraRibbonRendererProperties]}
  {emitterName=Fog, rendererClasses [&NiagaraSpriteRendererProperties, &NiagaraMeshRendererProperties]}
  {emitterName=SnowDust001, rendererClasses [&NiagaraSpriteRendererProperties]}
 ]
 userVariables []
}
